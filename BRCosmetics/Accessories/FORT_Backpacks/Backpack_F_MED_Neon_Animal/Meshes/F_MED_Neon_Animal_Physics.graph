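# Physics asset for the Neon Animal backpack cosmetic's skeletal mesh, supporting a dangling ornament on the pack.
# A capsule body covers the pack, and three box bodies follow the dyn_ornament bone chain.
# Constraints link pack to ornament 01, 02 and 03 in sequence, each with moderate swing and tight twist limits so the ornament can sway without spinning.
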PhysicsAsset /BRCosmetics/Accessories/FORT_Backpacks/Backpack_F_MED_Neon_Animal/Meshes/F_MED_Neon_Animal_Physics
bodies {
 dyn_ornament_01 [{shapeType=Box, extentX=0.9275, extentY=0.8984, extentZ=4.7206, center=<2.3605 0 0>, rotation=<rot 90 0 0>}]
 dyn_ornament_02 [{shapeType=Box, extentX=6.1418, extentY=2.3015, extentZ=4.3059, center=<2 0 0>, rotation=<rot 90 0 0>}]
 dyn_ornament_03 [{shapeType=Box, extentX=7.1941, extentY=4.2446, extentZ=9.3136, center=<2.3113 0 0>, rotation=<rot 90 0 0>}]
 pack [{shapeType=Capsule, length=15, radius=9.8127, center=<-2.0577 3.8164 -0.2655>}]
}
constraints [
 {bone1Name=dyn_ornament_03, bone2Name=dyn_ornament_02, swing1LimitDegrees=45, swing2LimitDegrees=45, twistLimitDegrees=7.5}
 {bone1Name=dyn_ornament_01, bone2Name=pack, swing1LimitDegrees=45, swing2LimitDegrees=45, twistLimitDegrees=7.5}
 {bone1Name=dyn_ornament_02, bone2Name=dyn_ornament_01, swing1LimitDegrees=45, swing2LimitDegrees=45, twistLimitDegrees=7.5}
]
